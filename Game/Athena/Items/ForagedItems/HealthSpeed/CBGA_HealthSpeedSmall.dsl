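; CBGA_HealthSpeedSmall is a small health-and-speed foraged item in Athena that pawns can pick up and consume.
; On BeginPlay it calls the parent, then branches on bShouldRandomizeRotation to handle random world yaw and HeightOffset placement of its Repositioned component, so placed items look less uniform.
; Its interact and allowed-to-consume overrides always return true.
; The construction script only calls the parent, and no health or speed effect logic appears here.
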(event EventBeginPlay
  (|CallParentFunction)
  ;[1 1 1] Give the mesh a random rot when it starts that makes them feel more randomly placed.
  (if (Variables|Default|GetbShouldRandomizeRotation)
    (Transformation|SetWorldRotation (Variables|Default|GetRepositioned) (Math|Rotator|MakeRotator 0.0 0.0 (Math|Random|RandomFloatinRange 0.0 3.6e+02)) false false)
    (Transformation|SetRelativeLocation (Variables|Default|GetRepositioned) (Variables|Default|GetHeightOffset) false false)))

(fn BlueprintCanInteract
  (params (InteractingPawn :type Fort%20Pawn%20Object%20Reference) (InteractionBeingAttempted :type EInteractionBeingAttempted%20Enum :default "FirstInteraction") (InteractionType :type TInteractionType%20Enum :default "IT_NoInteraction"))
  (return true))

(fn OnInteractAllowedToConsume
  (params (InteractingPawn :type Fort%20Pawn%20Object%20Reference) (InteractionBeingAttempted :type EInteractionBeingAttempted%20Enum :default "FirstInteraction"))
  (return true))

(fn ConstructionScript
  (|CallParentFunction))
